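# Masked, two-sided subsurface material for the blob effect on the Genius backpack cosmetic, built for Niagara mesh particles.
# SplineThicken sets strand width, and dynamic parameters drive a RotateAboutAxis bend from rest direction toward a target position, weighted along V.
# PreviousFrameSwitch repeats the bend with the previous position for correct motion vectors.
# Power-warped UVs sample drop textures, and two smoothstepped GGX terms add hard stylized highlights tinted through particle color.
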Material /BRCosmetics/Accessories/FORT_Backpacks/Backpack_F_MED_Genius/FX/M_Backpack_Blob
graph {
 properties {
  bTangentSpaceNormal=false
  bUsedWithNiagaraMeshParticles=true
  blendMode=BLEND_Masked
  pixelDepthOffsetMode=PDOM_Legacy
  shadingModel=MSM_Subsurface
  twoSided=true
  usedShadingModels=Subsurface
 }
 nodes [
  n1 = ParticleColor()
  Desaturation(None=Multiply(A=VectorParameter() {parameterName=Color, defaultValue=<rgba 0.6 0.24 0.06 1>}.RGB, B=@n1.R), Fraction=@n1.G)
  Desaturation(None=Multiply(A=VectorParameter() {parameterName=Colorsss, defaultValue=<rgba 0.9 0.4013 0.045 1>}.RGB, B=@n1.R), Fraction=@n1.G)
  n2 = StaticBool() {value=true}
  n3 = MaterialFunctionCall(Float2=TextureCoordinate()) {materialFunction=&BreakOutFloat2Components}
  n4 = OneMinus(None=@n3.G)
  n5 = DynamicParameter() {parameterIndex=3, paramNames [Power, Bend, Width, ""]}
  n6 = MaterialFunctionCall(X=@n3.R, Y=OneMinus(None=Power(Base=@n4, Exp=@n5.Power))) {materialFunction=&MakeFloat2}
  n7 = MaterialFunctionCall(
   WidthBase=@n5.Width
   WidthTip=@n5.Width
   Expand U or V UV Channel=@n2
   AdditionalNormal=MaterialFunctionCall(
    Normal=TextureSample(UVs=@n6.Result) {samplerType=SAMPLERTYPE_Normal, texture=&/T_DropSeamless_N}.RGB
    Flatness=ScalarParameter() {defaultValue=0.2, parameterName=FFFFFF}
   ) {materialFunction=&FlattenNormal}.Result
   FlattenPixelNormal=@n2
  ) {materialFunction=&SplineThicken}
  n8 = SmoothStep(Value=MaterialFunctionCall!(LightDirctionIn, CameraVectorIn, Roughness=ScalarParameter() {defaultValue=0.6, parameterName=R2}, NormalIn=@n7.Normal) {materialFunction=/Game/Effects/Niagara/MasterMaterials/Functions/Common/FXCommon_GGX}.Result) {constMax=0.2, constMin=0.19, constValue=0.21}
  n9 = SkyAtmosphereLightDirection()
  n10 = SmoothStep(Value=MaterialFunctionCall(LightDirctionIn=Multiply(A=@n9) {constB=-1}, Roughness=ScalarParameter() {defaultValue=0.75, parameterName=R}, NormalIn=@n7.Normal) {materialFunction=&FXCommon_GGX}.Result) {constMax=0.2, constMin=0.19, constValue=0.21}
  n11 = Add(
   A=Multiply(A=Saturate(None=@n8), B=ScalarParameter() {defaultValue=0.7, parameterName=str222})
   B=Multiply(A=Saturate(None=@n10), B=ScalarParameter() {defaultValue=0.2, parameterName=str2})
  )
  n12 = Desaturation(None=VectorParameter() {parameterName=ColorH, defaultValue=<rgba 1 0.4 0 1>}.RGB, Fraction=@n1.G)
  MaterialFunctionCall(Emissive=Multiply(A=Multiply(A=@n11, B=@n12), B=@n1.A)) {materialFunction=&MF_FX_EmissiveAdjustments}
  n13 = MaterialFunctionCall(WidthBase=@n5.Width, WidthTip=@n5.Width, Expand U or V UV Channel=@n2) {materialFunction=&SplineThicken}
  n14 = DynamicParameter() {paramNames [PosX,PosY,PosZ,RestX]}
  n15 = AppendVector(A=@n14.PosX, B=@n14.PosX)
  n16 = DynamicParameter() {parameterIndex=1, paramNames [RestY,RestZ,MeshPosX,MeshPosY]}
  n17 = DynamicParameter() {parameterIndex=2, paramNames [MeshPosZ,PrePosX,PrePosY,PrePosZ]}
  n18 = AppendVector(A=AppendVector(A=@n16.MeshPosX, B=@n16.MeshPosX), B=@n17.MeshPosZ)
  n19 = Normalize(VectorInput=Subtract(A=AppendVector(A=@n15, B=@n14.PosZ), B=@n18))
  n20 = Normalize(VectorInput=AppendVector(A=AppendVector(A=@n14.RestX, B=@n16.RestY), B=@n16.RestZ))
  n21 = TextureCoordinate()
  n22 = Power(Base=OneMinus(None=MaterialFunctionCall(Float2=@n21) {materialFunction=&BreakOutFloat2Components}.G))
  Reroute(None=LinearInterpolate(
    A=@n13.Normal
    B=MaterialFunctionCall!(VectorToRotate=@n13.Normal, LookAtVector=@n19, RestingVector=@n20) {materialFunction=/Engine/Functions/Engine_MaterialFunctions01/Vectors/RotateVector}.Result
    Alpha=@n22
   ))
  n23 = OneMinus(None=Saturate(None=DotProduct(A=@n19, B=@n20)))
  n24 = Multiply(A=@n5.Bend, B=ScalarParameter() {defaultValue=0.2, parameterName=Rigidity})
  n25 = Add(A=@n13.WorldPositionOffset, B=WorldPosition() {worldPositionShaderOffset=WPT_ExcludeAllShaderOffsets}.XYZ)
  n26 = RotateAboutAxis(
   NormalizedRotationAxis=CrossProduct(A=@n20, B=@n19)
   RotationAngle=Multiply(A=@n22, B=Multiply(A=@n23, B=@n24))
   PivotPoint=@n18
   Position=@n25
  )
  n27 = AppendVector(A=@n17.PrePosX, B=@n17.PrePosX)
  n28 = Normalize(VectorInput=Subtract(A=AppendVector(A=@n27, B=@n17.PrePosZ), B=@n18))
  n29 = OneMinus(None=Saturate(None=DotProduct(A=@n28, B=@n20)))
  n30 = RotateAboutAxis(
   NormalizedRotationAxis=CrossProduct(A=@n20, B=@n28)
   RotationAngle=Multiply(A=@n22, B=Multiply(A=@n24, B=@n29))
   PivotPoint=@n18
   Position=@n25
  )
  Reroute(None=Add(A=PreviousFrameSwitch(Current Frame=@n26, Previous Frame=@n30), B=@n13.WorldPositionOffset))
  ScalarParameter() {defaultValue=0.5, parameterName=sss}
  TextureSample(UVs=@n6.Result) {texture=&/T_DropSeamless}
  VectorParameter() {parameterName=P, defaultValue=<rgba 0 0 0 1>}
 ]
}
