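# Niagara system for the Genius backpack cosmetic, growing drip-like strands from eight skeletal mesh sockets.
# Each particle picks a socket, using a counter and scratch module to avoid repeats, then swings as a lengthening spring pendulum until releasing late in life.
# Positions feed dynamic material parameters for sprite and mesh renderers, with color varying by time of day or front-end preview.
# A smoothed system value scales strands and kills particles by threshold.
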NiagaraSystem /BRCosmetics/Accessories/FORT_Backpacks/Backpack_F_MED_Genius/FX/NS_Backpack_Genius
dependencies {
 usedDataInterfaces [&NiagaraDataInterfaceCurve]
 usedDynamicInputs []
 usedModules [&SystemState, &InterpolateOverTime]
}
emitters {Empty {
  emitterUpdateScript [
   &EmitterState(
    Inactive Response=::NewEnumerator1
    Life Cycle Mode=::NewEnumerator1
    Loop Behavior=::NewEnumerator1
    Loop Duration Mode=::NewEnumerator1
    UseLoopDelay=false
   )
   &SpawnRate(SpawnRate=0.5, Use Spawn Probability=false)
   &SpawnBurst_Instantaneous(Spawn Count=8, Use Spawn Probability=false, UseLoopCountLimit=false) {enabled=false}
  ]
  particleSpawnScript [
   &InitializeParticle(
    Color Mode=::NewEnumerator0
    Lifetime Max=6
    Lifetime Min=4
    Lifetime Mode=::NewEnumerator1
    Sprite Size Mode=::NewEnumerator3
    Uniform Sprite Size=1
    UsePositionOffset=false
    Write Material Random=true
   )
   /SetVariables(Particles.SocketIndex:NiagaraInt32=&RandomRangeInt(Minimum=0, Maximum=7)) {isSetParameters=true}
   /SetVariables(Particles.SocketIndex:NiagaraInt32=&MakeCustomIntFromBool(
     Bool=&SetBoolByIntComparison(
      A=&self:ScratchDynamicInput(New Simple Counter:NiagaraDataInterfaceSimpleCounter=$User.SC)
      B=$Particles.SocketIndex
      Comparison Type=::NewEnumerator4
     )
     True INT=&ModuloInt(
      A=&Add_Integer(A=$Particles.SocketIndex, B=&RandomRangeInt(Minimum=1, Maximum=7))
      B=8
     )
     False INT=$Particles.SocketIndex
    )) {isSetParameters=true}
   &self:AvoidRepet(New Simple Counter:NiagaraDataInterfaceSimpleCounter=$User.SC, New int32:NiagaraInt32=$Particles.SocketIndex) {module=AvoidRepet}
   &SampleSkeletalMesh(
    AdditionalYawPitchRoll=false
    Bone / Socket Index:NiagaraInt32=$Particles.SocketIndex
    Mesh Sampling Type:ENiagaraSkelSamplingModeFull=::NewEnumerator5
    Skeletal Mesh:NiagaraDataInterfaceSkeletalMesh=$User.Target
    Socket Sampling Mode:ENiagaraSocketSamplingMode=::NewEnumerator5
   ) {module=SampleSkeletalMesh001}
   /SetVariables(Particles.Position=$Particles.SampleSkeletalMesh001.SampledPosition) {isSetParameters=true}
   /SetVariables() {isSetParameters=true}
   /SetVariables(Particles.RandomBase:NiagaraFloat=&RandomRangeFloat(Minimum=0, Maximum=1)) {isSetParameters=true}
   /SetVariables(Particles.Length:NiagaraFloat=&Lerp_Float(A=4, B=8, Alpha=$Particles.RandomBase)) {isSetParameters=true}
   /SetVariables(Particles.Width:NiagaraFloat=&Lerp_Float(A=1.5, B=2, Alpha=$Particles.RandomBase)) {isSetParameters=true}
   /SetVariables(Particles.Connect=true) {isSetParameters=true}
  ]
  particleUpdateScript [
   &ParticleState()
   /SetVariables(Particles.Connect:NiagaraBool=&SetBoolByFloatComparison(A=$Particles.NormalizedAge, B=0.7, Comparison Type=::NewEnumerator6)) {isSetParameters=true}
   /SetVariables(Particles.Length:NiagaraFloat=&Multiply_Float(
     A=$Particles.Initial.Length
     B=&FloatFromCurve(FloatCurve=DataInterface() {propertyValues {Curve {keys [
          {interpMode=RCIM_Cubic, leaveTangent=3.9725, tangentMode=RCTM_Break, time=0, value=0}
          {arriveTangent=3.6601, interpMode=RCIM_Cubic, tangentMode=RCTM_Break, time=0.7, value=1}
          {interpMode=RCIM_Cubic, tangentMode=RCTM_Break, time=0.73, value=0.65}
         ]}}})
    )) {isSetParameters=true}
   &SampleSkeletalMesh(
    AdditionalYawPitchRoll=false
    Bone / Socket Index:NiagaraInt32=$Particles.SocketIndex
    Mesh Sampling Type:ENiagaraSkelSamplingModeFull=::NewEnumerator5
    Skeletal Mesh:NiagaraDataInterfaceSkeletalMesh=$User.Target
    Socket Sampling Mode:ENiagaraSocketSamplingMode=::NewEnumerator5
   )
   /SetVariables(Particles.tmpForce:Vector3f=$Transient.PhysicsForce) {isSetParameters=true}
   &PendulumSetup(
    Module.Enable Spring Driver=true
    Pendulum Length:NiagaraFloat=$Particles.Length
    Pendulum Pivot:Vector3f=$Particles.SampleSkeletalMesh.SampledPosition
    Rigid Pendulum=false
    "Spring Driven Constraint (Tightness)":NiagaraFloat=&RandomRangeFloat(Minimum=1.6, Maximum=2.2)
   )
   /SetVariables(Transient.PhysicsForce:Vector3f=&MakeCustomVectorFromBool(Bool=$Particles.Connect, True Vector=$Transient.PhysicsForce, False Vector=$Particles.tmpForce)) {isSetParameters=true}
   &GravityForce()
   &SolveForcesAndVelocity(Clamp Velocity=false, Limit Acceleration=false)
   /SetVariables(Particles.TmpP:NiagaraPosition=$Particles.Position) {isSetParameters=true}
   &PendulumConstraint()
   /SetVariables(Particles.Position=&MakeCustomVectorFromBool(
     Bool=$Particles.Connect
     True Vector=Linked() {linkedVariable {name=Particles.Position, type=NiagaraPosition}}
     False Vector=Linked() {linkedVariable {name=Particles.TmpP, type=NiagaraPosition}}
    )) {isSetParameters=true}
   &CalculateAccurateVelocity()
   &self:Avoid_Intersection(RootP:Vector3f=$Particles.SampleSkeletalMesh.SampledPosition) {module=Avoid_Intersection}
   /SetVariables(Particles.MeshOrientation=&FindQuatBetweenTwoVectors(
     From Vector=<0 0 1>
     To Vector=&MakeCustomVectorFromBool(
      Bool=$Particles.Connect
      True Vector=&NormalizeVector(
       Vector To Normalize=&SubtractVectorFromPosition(Position=$Particles.Position, Vector=$Particles.SampleSkeletalMesh.SampledPosition)
       Safe Normalize=true
       Safe Fallback Threshold=0
       Safe Fallback Vector=<0 0 0.001>
      )
      False Vector=&Lerp_Vector(
       A=&NormalizeVector(
        Vector To Normalize=&SubtractVectorFromPosition(Position=$Particles.Position, Vector=$Particles.SampleSkeletalMesh.SampledPosition)
        Safe Normalize=true
        Safe Fallback Threshold=0
        Safe Fallback Vector=<0 0 0.001>
       )
       B=<0 0 -1>
       Alpha=&FloatFromCurve(FloatCurve=DataInterface() {propertyValues {Curve {keys [<at 0.7 0>,<at 0.75 1>]}}})
      )
     )
    )) {isSetParameters=true}
   /SetVariables(Particles.MeshPPos:NiagaraPosition=&AddVectorToPosition(
     Position=$Particles.Position
     Vector=&Multiply_VectorByFloat(
      Vector=&RotateVectorByQuaternion(Vector To Rotate=<0 0 -1>, Quaternion=$Particles.MeshOrientation)
      Float=$Particles.Length
     )
    )) {isSetParameters=true}
   /SetVariables(Particles.ScaleCurve:NiagaraFloat=&FloatFromCurve(FloatCurve=DataInterface() {propertyValues {Curve {keys [<at 0.8 1>,<at 1 0>]}}})) {isSetParameters=true}
   /SetVariables(Particles.Scale=&MakeVector(
     X=0.001
     Y=0.001
     Z=&Multiply_Float(
      A=&Multiply_Float(A=$Particles.Length, B=$System.InterpolateOverTime.InterpolateOverTime_Float.Moving Average)
      B=$Particles.ScaleCurve
     )
    )) {isSetParameters=true}
   /SetVariables(Particles.PositionForRender:Vector3f=$Particles.SampleSkeletalMesh.SampledPosition) {isSetParameters=true}
   &DynamicMaterialParameters(
    Index 0 Param 1=&MakeFloatFromVector(Vector=Linked() {linkedVariable {name=Particles.Position, type=NiagaraPosition}}, Channel=::NewEnumerator0)
    Index 0 Param 2=&MakeFloatFromVector(Vector=Linked() {linkedVariable {name=Particles.Position, type=NiagaraPosition}}, Channel=::NewEnumerator1)
    Index 0 Param 3=&MakeFloatFromVector(Vector=Linked() {linkedVariable {name=Particles.Position, type=NiagaraPosition}}, Channel=::NewEnumerator2)
    Index 0 Param 4=-0.001
    Index 1 Param 1=0.001
    Index 1 Param 2=-1
    Index 1 Param 3=&MakeFloatFromVector(Vector=$Particles.PositionForRender, Channel=::NewEnumerator0)
    Index 1 Param 4=&MakeFloatFromVector(Vector=$Particles.PositionForRender, Channel=::NewEnumerator1)
    Index 2 Param 1=&MakeFloatFromVector(Vector=$Particles.PositionForRender, Channel=::NewEnumerator2)
    Index 2 Param 2=&MakeFloatFromVector(Vector=Linked() {linkedVariable {name=Particles.Previous.Position, type=NiagaraPosition}}, Channel=::NewEnumerator0)
    Index 2 Param 3=&MakeFloatFromVector(Vector=Linked() {linkedVariable {name=Particles.Previous.Position, type=NiagaraPosition}}, Channel=::NewEnumerator1)
    Index 2 Param 4=&MakeFloatFromVector(Vector=Linked() {linkedVariable {name=Particles.Previous.Position, type=NiagaraPosition}}, Channel=::NewEnumerator2)
    Index 3 Param 1=&FloatFromCurve(FloatCurve=DataInterface() {propertyValues {Curve {keys [
         {interpMode=RCIM_Cubic, tangentMode=RCTM_Break, time=0, value=1}
         <at 0.1 1>
         {arriveTangent=1.6518, leaveTangent=20.8249, time=0.7, value=1}
         {interpMode=RCIM_Cubic, tangentMode=RCTM_Break, time=0.73, value=2.2}
        ]}}})
    Index 3 Param 2=&Lerp_Float(
     A=0
     B=&MakeCustomFloatFromBool(
      Bool=$Particles.Connect
      True Float=&MakeCustomFloatFromBool(Bool=$User.bIsFrontEndPreview, True Float=1.5, False Float=1)
      False Float=0
     )
     Alpha=$System.InterpolateOverTime.InterpolateOverTime_Float.Moving Average
    )
    Index 3 Param 3=&FloatFromCurve(
     FloatCurve=DataInterface() {propertyValues {Curve {keys [
         {interpMode=RCIM_Cubic, leaveTangent=3.8259, tangentMode=RCTM_Break, time=0, value=0}
         {arriveTangent=-1.041, interpMode=RCIM_Cubic, leaveTangent=-1.041, tangentMode=RCTM_User, time=0.0939, value=0.5871}
         {interpMode=RCIM_Cubic, tangentMode=RCTM_Break, time=0.7, value=0.48}
         {interpMode=RCIM_Cubic, tangentMode=RCTM_Break, time=0.7389, value=0.4824}
        ]}}}
     Scale Curve=&Multiply_Float(A=&Multiply_Float(A=$Particles.Width, B=1), B=$Particles.ScaleCurve)
    )
    Module.Param0WriteEnabled=true
    Module.Param0WriteEnabled1=true
    Module.Param0WriteEnabled2=true
    Module.Param0WriteEnabled3=true
    Module.Param1WriteEnabled=true
    Module.Param1WriteEnabled1=true
    Module.Param1WriteEnabled2=true
    Module.Param1WriteEnabled3=true
    Module.Param2WriteEnabled=true
    Module.Param2WriteEnabled1=true
    Module.Param2WriteEnabled2=true
    Module.Param2WriteEnabled3=true
    Module.Param3WriteEnabled=true
    Module.Param3WriteEnabled1=true
    Module.Param3WriteEnabled2=true
    Module.Param3WriteEnabled3=true
    Write Parameter Index 1=true
    Write Parameter Index 2=true
    Write Parameter Index 3=true
   ) {module=DynamicMaterialParameters001}
   &CameraOffset(Camera Offset Amount=5)
   /SetVariables(Particles.Color=&MakeLinearColor(
     R:NiagaraFloat=&MakeCustomFloatFromBool(
      Bool=$User.bIsFrontEndPreview
      True Float=0.15
      False Float=&FNTimeOfDayScalar(Night Scale=0.6, Morning Scale=0.2, Day Scale=0.4, Evening Scale=0.2)
     )
     G:NiagaraFloat=&MakeCustomFloatFromBool(
      Bool=$User.bIsFrontEndPreview
      True Float=-0.5
      False Float=&FNTimeOfDayScalar(Night Scale=-0.4, Morning Scale=0.35, Day Scale=0, Evening Scale=0.35)
     )
     B:NiagaraFloat=0
     A:NiagaraFloat=&MakeCustomFloatFromBool(
      Bool=$User.bIsFrontEndPreview
      True Float=1
      False Float=&FNTimeOfDayScalar(Night Scale=0.5, Morning Scale=0.3, Day Scale=0.6, Evening Scale=0.3)
     )
    )) {isSetParameters=true}
   &KillParticles(Kill Particles=&SetBoolByFloatComparison(A=$System.InterpolateOverTime.InterpolateOverTime_Float.Moving Average, B=0.2, Comparison Type=::NewEnumerator6))
  ]
  rendererClasses [&NiagaraSpriteRendererProperties, &NiagaraMeshRendererProperties]
  renderers [
   {rendererClass=&NiagaraSpriteRendererProperties}
   {rendererClass=&NiagaraMeshRendererProperties, rendererIndex=1}
   {rendererClass=&NiagaraMeshRendererProperties, rendererIndex=2}
  ]
 }}
summary {
 systemName=NS_Backpack_Genius
 emitters [{emitterName=Empty, rendererClasses [&NiagaraSpriteRendererProperties, &NiagaraMeshRendererProperties]}]
 userVariables [
  User.bIsFrontEndPreview {type=NiagaraBool}
  User.bIsJump {type=NiagaraBool}
  User.SC {
   type=NiagaraDataInterfaceSimpleCounter
   description=""
   defaultValue {
    struct=NiagaraExt_VariableValue_DataInterface
    value {
     dataInterface=&self:NiagaraDataInterfaceSimpleCounter_0
     dataInterfaceClass=/Script/Niagara.NiagaraDataInterfaceSimpleCounter
    }
   }
  }
  User.Target {
   type=NiagaraDataInterfaceSkeletalMesh
   defaultValue {
    struct=NiagaraExt_VariableValue_DataInterface
    value {dataInterface=&self:NiagaraDataInterfaceSkeletalMesh_0}
   }
  }
 ]
}
